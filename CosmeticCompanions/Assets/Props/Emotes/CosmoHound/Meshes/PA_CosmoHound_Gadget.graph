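# Physics asset for the CosmoHound gadget, a prop mesh among the Cosmetic Companions emote assets.
# A single rotated capsule on the root bone covers the prop, with a near-zero length that makes it almost a large sphere.
# It has no other bodies or constraints, so it provides one coarse shape with no per-bone detail or jointed ragdoll.
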PhysicsAsset /CosmeticCompanions/Assets/Props/Emotes/CosmoHound/Meshes/PA_CosmoHound_Gadget
bodies {root [{shapeType=Capsule, length=0.505, radius=258.2391, rotation=<rot 0 -90 0>}]}
constraints []
